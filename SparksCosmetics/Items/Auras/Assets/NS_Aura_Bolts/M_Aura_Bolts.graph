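# M_Aura_Bolts is an unlit, additive material for Niagara sprites in the Bolts aura cosmetic.
# It samples the T_Aura_Bolts texture and multiplies its RGB by particle color and its alpha by particle alpha.
# This lets the particle system control the tint and fade of each sprite.
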Material /SparksCosmetics/Items/Auras/Assets/NS_Aura_Bolts/M_Aura_Bolts
graph {
 properties {bUsedWithNiagaraSprites=true, blendMode=BLEND_Additive, pixelDepthOffsetMode=PDOM_Legacy, shadingModel=MSM_Unlit, usedShadingModels=Unlit}
 nodes [
  n1 = TextureSample() {texture=&/T_Aura_Bolts}
  n2 = ParticleColor()
  Multiply(A=@n1.RGB, B=@n2.RGB)
  Multiply(A=@n1.A, B=@n2.A)
 ]
}
